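; BP_ColorPalette is an art-tool actor that displays a color palette, chosen by data table row, as a grid of instanced meshes.
; The construction script reads the row's dimensions and texture, and MapMeshesToTexture lays out centered instances with configurable spacing and size, flat or vertical.
; Each instance gets normalized grid coordinates as custom data, alongside a dynamic M_PalettePicker material receiving the palette texture and roughness.
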(event EventBeginPlay)

(event Collision|EventActorBeginOverlap
  (params (OtherActor :type Actor%20Object%20Reference)))

(event EventTick
  (params (DeltaSeconds :type Float%20%28single-precision%29)))

(fn MapMeshesToTexture
  (params (X :type Integer) (Y :type Integer))
  (bind _instancedstaticmesh (Variables|Default|GetInstancedStaticMesh))
  (bind _instancestransformarray (Variables|Default|GetInstancesTransformArray))
  (LocalVariables|SetMaxX (- X 1))
  (LocalVariables|SetMaxY (- Y 1))
  ;[1 1 1] Rows
  (for _index (range (+ (LocalVariables|GetMaxY) 1))
    (LocalVariables|SetLocalY _index)
    ;[1 1 1] Columns
    (for _index_2 (range (+ (LocalVariables|GetMaxX) 1))
      (LocalVariables|SetLocalX _index_2)
      (bind _returnvalue (Utilities|Array|Add _instancestransformarray (Math|Transform|MakeTransform (Math|Vector|vector+vector (- (* (select (Variables|Default|GetbVertical) (Math|Vector|MakeVector 0.0 (Math|Conversions|ToFloat(Integer) (LocalVariables|GetLocalX)) (Math|Conversions|ToFloat(Integer) (LocalVariables|GetLocalY))) (Math|Vector|MakeVector (Math|Conversions|ToFloat(Integer) (LocalVariables|GetLocalY)) (Math|Conversions|ToFloat(Integer) (LocalVariables|GetLocalX)))) (* (Variables|Default|GetSpacing) -1.0)) (* (select (Variables|Default|GetbVertical) (Math|Vector|MakeVector 0.0 (Math|Conversions|ToFloat(Integer) (LocalVariables|GetMaxX)) (Math|Conversions|ToFloat(Integer) (LocalVariables|GetMaxY))) (Math|Vector|MakeVector (Math|Conversions|ToFloat(Integer) (LocalVariables|GetMaxY)) (Math|Conversions|ToFloat(Integer) (LocalVariables|GetMaxX)))) (* (Variables|Default|GetSpacing) -0.5))) (Math|Vector|MakeVector 0.0 0.0 (select (Variables|Default|GetbVertical) (+ (/ (* (LocalVariables|GetMaxY) (Variables|Default|GetSpacing)) 2.0) (Variables|Default|GetSpacing)) (* (Variables|Default|GetSize) 2e+01)))) "0, 0, 0" (Math|Conversions|ToVector(Float) (* (Variables|Default|GetSize) 0.4)))))
      (bind _returnvalue_2 (Components|InstancedStaticMesh|AddInstance (Variables|Default|GetInstancedStaticMesh) (Utilities|Array|Get(acopy) _instancestransformarray _returnvalue) false))
      (Components|InstancedStaticMesh|SetCustomDataValue _instancedstaticmesh _returnvalue_2 0 (Math|Float|MapRangeClamped (/ (Math|Conversions|ToFloat(Integer) (LocalVariables|GetLocalX)) (Math|Conversions|ToFloat(Integer) (LocalVariables|GetMaxX))) 0.0 1.0 0.125 0.875) false)
      (Components|InstancedStaticMesh|SetCustomDataValue _instancedstaticmesh _returnvalue_2 1 (Math|Float|MapRangeClamped (/ (Math|Conversions|ToFloat(Integer) (LocalVariables|GetLocalY)) (Math|Conversions|ToFloat(Integer) (LocalVariables|GetMaxY))) 0.0 1.0 0.125 0.875) false))))

(fn ConstructionScript
  (bind (_datatable _rowname) (Utilities|Struct|BreakDataTableRowHandle (Variables|Default|GetPalette)))
  (DataTable|GetDataTableRowNames "/Game/ArtTools/ColorPalette/ColorPalettes_Table.ColorPalettes_Table")
  (bind _returnvalue (Utilities|GetDataTableRow "/Game/ArtTools/ColorPalette/ColorPalettes_Table.ColorPalettes_Table" _rowname))
  (CallFunction|MapMeshestoTexture (Math|Float|Round (Math|Vector2D|BreakVector2D (Utilities|Struct|BreakStructColorPalettes _returnvalue))) (Math|Float|Round (.Y (Math|Vector2D|BreakVector2D (Utilities|Struct|BreakStructColorPalettes _returnvalue)))))
  (bind _returnvalue_2 (Rendering|Material|CreateDynamicMaterialInstance (Variables|Default|GetInstancedStaticMesh) 0 "/Game/ArtTools/ColorPalette/Materials/M_PalettePicker.M_PalettePicker" "None" :!class /Script/Engine.PrimitiveComponent))
  (Rendering|Material|SetTextureParameterValue _returnvalue_2 "Color" (.Texture_10_ED254AEA4279F825741648A12485D566 (Utilities|Struct|BreakStructColorPalettes _returnvalue)))
  (Rendering|Material|SetScalarParameterValue _returnvalue_2 "Roughness" (Variables|Default|GetRoughness) :!class /Script/Engine.MaterialInstanceDynamic))
