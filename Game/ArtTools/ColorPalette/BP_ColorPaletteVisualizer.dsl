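; Editor art-tool actor that displays a biome's color palette as a labeled grid of instanced meshes, built in the construction script.
; It reads the selected row from the ColorPalette_Biomes data table and lays out one instance per entry in a near-square grid, horizontal or vertical.
; Color and roughness travel as per-instance custom data to M_Palette, whose dynamic instance can override roughness.
; Colored text components label each swatch with its name or linear color value.
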(event EventBeginPlay)

(event Collision|EventActorBeginOverlap
  (params (OtherActor :type Actor%20Object%20Reference)))

(event EventTick
  (params (DeltaSeconds :type Float%20%28single-precision%29)))

(fn GrabColorDataFromTable
  (params (Palette :type Array%20of%20Struct%20Palette%20Element%20Structures) (Modulus :type Integer))
  (bind _instancedstaticmesh (Variables|Default|GetInstancedStaticMesh))
  (bind _instancestransformarray (Variables|Default|GetInstancesTransformArray))
  (LocalVariables|SetMaxX (- Modulus 1))
  (LocalVariables|SetMaxY Modulus)
  (for (_array_element _array_index) Palette
    (LocalVariables|SetLocalX (Math|Integer|%25%28Integer%29 _array_index Modulus))
    (LocalVariables|SetLocalY (/ _array_index Modulus))
    (bind _returnvalue (Utilities|Array|Add _instancestransformarray (Math|Transform|MakeTransform (Math|Vector|vector+vector (- (* (select (Variables|Palette|GetbVertical) (Math|Vector|MakeVector 0.0 (Math|Conversions|ToFloat(Integer) (LocalVariables|GetLocalX)) (Math|Conversions|ToFloat(Integer) (LocalVariables|GetLocalY))) (Math|Vector|MakeVector (Math|Conversions|ToFloat(Integer) (LocalVariables|GetLocalY)) (Math|Conversions|ToFloat(Integer) (LocalVariables|GetLocalX)))) (* (Variables|Palette|GetSpacing) -1.0)) (* (select (Variables|Palette|GetbVertical) (Math|Vector|MakeVector 0.0 (Math|Conversions|ToFloat(Integer) (LocalVariables|GetMaxX)) (Math|Conversions|ToFloat(Integer) (LocalVariables|GetMaxY))) (Math|Vector|MakeVector (Math|Conversions|ToFloat(Integer) (LocalVariables|GetMaxY)) (Math|Conversions|ToFloat(Integer) (LocalVariables|GetMaxX)))) (* (Variables|Palette|GetSpacing) -0.5))) (Math|Vector|MakeVector 0.0 0.0 (* (Variables|Palette|GetSize) 2e+01))) "0, 0, 0" (Math|Conversions|ToVector(Float) (* (Variables|Palette|GetSize) 0.4)))))
    (bind _returnvalue_2 (Components|InstancedStaticMesh|AddInstance (Variables|Default|GetInstancedStaticMesh) (Utilities|Array|Get(acopy) _instancestransformarray _returnvalue) false))
    (Components|InstancedStaticMesh|SetCustomDataValue _instancedstaticmesh _returnvalue_2 0 (Math|Vector4|BreakVector4 (Utilities|LinearColortoVector4 (.Color_7_FBE6A81441C013EB415E4487B1902C97 (Utilities|Struct|BreakStructPaletteElement _array_element)))) false)
    (Components|InstancedStaticMesh|SetCustomDataValue _instancedstaticmesh _returnvalue_2 1 (.Y (Math|Vector4|BreakVector4 (Utilities|LinearColortoVector4 (.Color_7_FBE6A81441C013EB415E4487B1902C97 (Utilities|Struct|BreakStructPaletteElement _array_element))))) false)
    (Components|InstancedStaticMesh|SetCustomDataValue _instancedstaticmesh _returnvalue_2 2 (.Z (Math|Vector4|BreakVector4 (Utilities|LinearColortoVector4 (.Color_7_FBE6A81441C013EB415E4487B1902C97 (Utilities|Struct|BreakStructPaletteElement _array_element))))) false)
    (Components|InstancedStaticMesh|SetCustomDataValue _instancedstaticmesh _returnvalue_2 3 (.Roughness_18_E5EDC9DA47340EF4769095871E044C7D (Utilities|Struct|BreakStructPaletteElement _array_element)) false)))

(fn ConstructionScript
  (bind _biome (Variables|Palette|GetBiome))
  (bind (_datatable _rowname) (Utilities|Struct|BreakDataTableRowHandle _biome))
  (Utilities|Struct|SetmembersinDataTableRowHandle _biome "/Game/ArtTools/ColorPalette/ColorPalette_Biomes.ColorPalette_Biomes")
  (bind _returnvalue (Utilities|GetDataTableRow "/Game/ArtTools/ColorPalette/ColorPalette_Biomes.ColorPalette_Biomes" _rowname))
  (CallFunction|GrabColorDatafromTable (Utilities|Struct|BreakStructPaletteBiome _returnvalue) (Math|Float|Ceil (Math|Float|Sqrt (Math|Conversions|ToFloat(Integer) (Utilities|Array|Length (Utilities|Struct|BreakStructPaletteBiome _returnvalue))))))
  (bind _returnvalue_2 (Rendering|Material|CreateDynamicMaterialInstance (Variables|Default|GetInstancedStaticMesh) 0 "/Game/ArtTools/ColorPalette/Materials/M_Palette.M_Palette" "None" :!class /Script/Engine.PrimitiveComponent))
  (Rendering|Material|SetScalarParameterValue _returnvalue_2 "OverrideRoughness" (Math|Conversions|ToFloat(Boolean) (Variables|Palette|GetbUse%20BP%20Roughness)) :!class /Script/Engine.MaterialInstanceDynamic)
  (Rendering|Material|SetScalarParameterValue _returnvalue_2 "Roughness" (Variables|Palette|GetRoughness) :!class /Script/Engine.MaterialInstanceDynamic)
  (for (_array_element _array_index) (Variables|Default|GetInstancesTransformArray)
    ;[1 1 1] Add Name
    (bind _returnvalue_3 (AddComponent|NODE_AddTextRenderComponent-0 :RelativeTransform (Math|Transform|MakeTransform (Math|Vector|vector+vector (.location _array_element) (Math|Vector|MakeVector 0.0 0.0 (/ (Variables|Palette|GetSpacing) 4.0))) (.rotation _array_element) (.scale _array_element))))
    (Rendering|Components|TextRender|SetText _returnvalue_3 (select (Variables|Palette|GetbShowLinearColorValue) (Utilities|ColortoText (.Color_7_FBE6A81441C013EB415E4487B1902C97 (Utilities|Struct|BreakStructPaletteElement (Utilities|Array|Get(acopy) (Utilities|Struct|BreakStructPaletteBiome _returnvalue) _array_index))) (Variables|Palette|GetDecimals)) (Utilities|Text|ToText(String) (Utilities|Struct|BreakStructPaletteElement (Utilities|Array|Get(acopy) (Utilities|Struct|BreakStructPaletteBiome _returnvalue) _array_index)))))
    (Rendering|Components|TextRender|SetHorizontalAlignment _returnvalue_3 "EHTA_Center")
    (Rendering|Components|TextRender|SetTextRenderColor _returnvalue_3 (Math|Conversions|ToColor(LinearColor) (.Color_7_FBE6A81441C013EB415E4487B1902C97 (Utilities|Struct|BreakStructPaletteElement (Utilities|Array|Get(acopy) (Utilities|Struct|BreakStructPaletteBiome _returnvalue) _array_index))) false))))
