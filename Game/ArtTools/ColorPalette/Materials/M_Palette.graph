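# M_Palette is a Default Lit material for instanced static meshes that takes its color and roughness from per-instance custom data, under the ColorPalette art tools.
# Custom data slots 0 to 2 pass through vertex interpolators and are appended into an RGB value.
# Slot 3 is interpolated likewise, then lerped toward a Roughness parameter by an OverrideRoughness scalar, allowing a shared roughness to replace per-instance values.
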Material /Game/ArtTools/ColorPalette/Materials/M_Palette
graph {
 properties {bUsedWithInstancedStaticMeshes=true, pixelDepthOffsetMode=PDOM_Legacy, usedShadingModels=Default Lit}
 nodes [
  LinearInterpolate(
   A=VertexInterpolator(VS=PerInstanceCustomData!(DefaultValue) {dataIndex=3}).PS
   B=ScalarParameter() {defaultValue=0.95, parameterName=Roughness, sliderMax=1}
   Alpha=ScalarParameter() {parameterName=OverrideRoughness}
  )
  n1 = PerInstanceCustomData()
  n2 = PerInstanceCustomData() {dataIndex=1}
  n3 = PerInstanceCustomData() {dataIndex=2}
  Reroute(None=MaterialFunctionCall(R=VertexInterpolator(VS=@n1).PS, G=VertexInterpolator(VS=@n2).PS, B=VertexInterpolator(VS=@n3).PS) {materialFunction=&AppendMany}.RGB)
 ]
}
