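# M_PaletteBackdrop is a Default Lit material in the ColorPalette art tools, by its name a backdrop for viewing palettes.
# A scalar parameter, Dark, linearly interpolates between white and black constants, so material instances can set the grey level, with white as the default.
# A separate zero constant sits alongside the blend, and no textures are used.
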Material /Game/ArtTools/ColorPalette/Materials/M_PaletteBackdrop
graph {
 properties {pixelDepthOffsetMode=PDOM_Legacy, usedShadingModels=Default Lit}
 nodes [
  Constant()
  LinearInterpolate(A=Constant() {r=1}, B=Constant(), Alpha=ScalarParameter() {parameterName=Dark})
 ]
}
