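# M_PalettePicker is a Default Lit art-tool material for instanced static meshes that picks a color from a palette texture per instance.
# Two per-instance custom data values pass through vertex interpolators and are appended into UV coordinates.
# Those UVs sample the Color texture parameter, which defaults to the general palette, so instances can share one material yet show different swatches.
# A scalar Roughness parameter defaults to a nearly matte value.
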Material /Game/ArtTools/ColorPalette/Materials/M_PalettePicker
graph {
 properties {bUsedWithInstancedStaticMeshes=true, pixelDepthOffsetMode=PDOM_Legacy, usedShadingModels=Default Lit}
 nodes [
  ScalarParameter() {defaultValue=0.95, parameterName=Roughness, sliderMax=1}
  n1 = PerInstanceCustomData!(DefaultValue)
  n2 = PerInstanceCustomData() {dataIndex=1}
  TextureSampleParameter2D(UVs=AppendVector(A=VertexInterpolator(VS=@n1).PS, B=VertexInterpolator(VS=@n2).PS)) {parameterName=Color, texture=/Game/ArtTools/ColorPalette/Textures/T_FN_Palette_General}
 ]
}
